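# Glass material for the Spy Tech Hacker character body, showing an animated holographic display of scanlines and scrolling masks over a glossy surface.
# Pre-skinned XZ position drives scanlines and a panned mask, with texture-varied scroll speed, tinted and combined with a gradient glow for emissive.
# Normal-mapped Fresnel sets opacity between character PreFX and PostFX functions.
# The sun/moon reflection highlight and Refraction parameter are left unconnected.
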Material /BRCosmetics/Characters/Player/Male/Medium/Bodies/M_MED_Spy_Tech_Hacker/Materials/M_Spy_Tech_Hacker_Glass
graph {
 properties {
  bScreenSpaceReflections=true
  bUseMaterialAttributes=true
  bUsedWithMeshDeformer=true
  bUsedWithMorphTargets=true
  bUsedWithSkeletalMesh=true
  pixelDepthOffsetMode=PDOM_Legacy
  refractionMethod=RM_IndexOfRefraction
  translucencyDirectionalLightingIntensity=5
  translucencyLightingMode=TLM_SurfacePerPixelLighting
  usedShadingModels=Default Lit
 }
 nodes [
  n1 = PreSkinnedPosition()
  n2 = Abs(None=ScalarParameter() {defaultValue=20, parameterName=Texture Size})
  n3 = ComponentMask(None=Divide(A=VertexInterpolator(VS=@n1).PS, B=Multiply(A=@n2) {constB=-1})) {b=true, r=true}
  n4 = ScalarParameter() {defaultValue=2, group=Customization_Holographic, parameterName=Holo_ Scanlines Thickness, sliderMax=1, sliderMin=0.01, sortPriority=-199}
  n5 = Power(
   Base=MaterialFunctionCall!(
    ScanLine_Tiling_1=ScalarParameter() {defaultValue=5, parameterName=Holo_ Scanline Tiling 1}
    Scanline_PannerSpeed_1
    ScanLine_Tiling=ScalarParameter() {defaultValue=50, parameterName=Holo_ Scanline Tiling}
    Scanline_PannerSpeed
    UVs=@n3
   ) {materialFunction=/Game/Weapons/WeaponSkins/MaterialFunction/ScanLines}.Result
   Exp=@n4
  )
  n6 = TextureSample(UVs=Multiply(A=ScalarParameter() {defaultValue=1, parameterName=Text Scroll BreakupTiling}, B=TextureCoordinate())) {
   samplerType=SAMPLERTYPE_LinearGrayscale
   texture=/BRCosmetics/Weapons/FORT_Melee/Pickaxe_Spy_Tech_Hacker_Male/Textures/T_Pickaxe_Spy_Tech_Hacker_Male_TextSpeed
  }
  n7 = Multiply(A=@n6.R, B=ScalarParameter() {defaultValue=-0.21, parameterName=Panning Speed})
  n8 = Constant()
  n9 = TextureSample(UVs=Panner(Coordinate=@n3, Speed=AppendVector(A=@n7, B=@n8)) {speedY=1}) {
   samplerType=SAMPLERTYPE_Masks
   texture=/BRCosmetics/Weapons/FORT_Melee/Pickaxe_Spy_Tech_Hacker_Male/Textures/T_Pickaxe_Spy_Tech_Hacker_Male_Masks
  }
  n10 = Multiply(A=@n9.B, B=VectorParameter() {parameterName=dawsdadf, defaultValue=<rgba 0.1833 0.3438 0 1>}.RGB)
  n11 = ScalarParameter() {defaultValue=0.25, parameterName=sfasfa}
  n12 = ScalarParameter() {defaultValue=3, parameterName=fasfaf}
  n13 = Subtract(
   A=MaterialFunctionCall(UV Channel=@n3) {materialFunction=&LinearGradient}.VGradient
   B=ScalarParameter() {defaultValue=-7.88, parameterName=fagff}
  )
  n14 = Saturate(None=Power(Base=LinearInterpolate(A=@n11, B=@n12, Alpha=@n13), Exp=ScalarParameter() {defaultValue=3, parameterName=GlowPower}))
  n15 = LinearInterpolate(A=ScalarParameter() {defaultValue=-0.02, parameterName=gawgf}, B=ScalarParameter() {defaultValue=1, parameterName=fasfsdfd}, Alpha=@n14)
  n16 = LinearInterpolate(
   A=VectorParameter() {group=Emissive, parameterName=EmissiveColor Edges, defaultValue=<rgba 0 0.0109 0.0208 1>}.RGB
   B=VectorParameter() {group=Emissive, parameterName=EmissiveColor Core, defaultValue=<rgba 0.3864 0.7379 0 1>}.RGB
   Alpha=@n14
  )
  n17 = ScalarParameter() {defaultValue=0.1, parameterName=Brightness}
  n18 = Multiply(
   A=Add(
    A=OneMinus(None=@n5)
    B=ScalarParameter() {defaultValue=0.5, group=Customization_Holographic, parameterName=Holo_ Scanlines Additional Brightness, sortPriority=-197}
   )
   B=Add(A=Multiply(A=@n10, B=@n15), B=Multiply(A=@n16, B=@n17))
  )
  n19 = CollectionParameter() {
   collection=&FortniteMaterialParameters
   parameterName=SunAndMoonModelDirectionalVector
  }
  n20 = ComponentMask(None=@n19) {b=true, g=true, r=true}
  n21 = ReflectionVectorWS()
  n22 = Power(Base=Clamp(None=DotProduct(A=@n20, B=@n21)), Exp=ScalarParameter() {defaultValue=100, parameterName=Spec Highlight Power}) {constExponent=20}
  n23 = ScalarParameter() {defaultValue=1, parameterName=Spec Intense}
  Add(
   A=@n18
   B=Multiply(A=VectorParameter() {parameterName=Reflection Color, defaultValue=<rgba 1.25 1.25 1.25 1>}.RGB, B=Multiply(A=@n22, B=@n23))
  )
  Constant3Vector() {constant=<rgba 0 0 1 0>}
  n24 = MaterialFunctionCall() {materialFunction=&BaseFN_CharacterFunction_PreFX}
  n25 = Saturate(None=@n18)
  n26 = TextureSampleParameter2D() {
   parameterName=Normal Map
   samplerType=SAMPLERTYPE_Normal
   texture=/BRCosmetics/Characters/Player/Male/Medium/Bodies/M_MED_Spy_Tech_Hacker/Textures/T_M_MED_Spy_Tech_Hacker_Acc_N
  }
  n27 = BreakMaterialAttributes!(Attr=@n24.Result)
  n28 = ScalarParameter() {defaultValue=5, parameterName=Fresnel Exponent}
  n29 = Transform(None=@n26.RGB)
  n30 = LinearInterpolate(
   A=ScalarParameter() {defaultValue=0.95, parameterName=Fresnel Inner Transparency}
   B=ScalarParameter() {defaultValue=0.15, parameterName=Fresnel Outer Transparency}
   Alpha=Saturate(None=Fresnel(ExponentIn=@n28, Normal=@n29))
  )
  MaterialFunctionCall(
   AO=@n24.AO
   Distance Blend=@n24.AO
   Fresnel=@n24.AO
   Material=MakeMaterialAttributes!(
    BaseColor=VectorParameter() {parameterName=Base Color}.RGB
    Metallic=Constant()
    Specular=Constant() {r=1}
    Roughness=Constant() {r=0.05}
    Anisotropy
    EmissiveColor=@n25
    Opacity=@n30
    OpacityMask=@n30
    Normal=@n26.RGB
    Tangent
    WorldPositionOffset=@n27.WorldPositionOffset
    SubsurfaceColor=@n27.WorldPositionOffset
    ClearCoat=@n27.WorldPositionOffset
    ClearCoatRoughness=@n27.WorldPositionOffset
    AmbientOcclusion=@n27.WorldPositionOffset
    Refraction=@n27.WorldPositionOffset
    CustomizedUVs_0
    CustomizedUVs_1
    CustomizedUVs_2
    CustomizedUVs_3
    CustomizedUVs_4
    CustomizedUVs_5
    CustomizedUVs_6
    CustomizedUVs_7
    PixelDepthOffset=@n27.WorldPositionOffset
    ShadingModel=@n27.WorldPositionOffset
    Displacement
   )
  ) {materialFunction=&BaseFN_CharacterFunction_PostFX}
  Multiply(A=@n25, B=@n30)
  ScalarParameter() {defaultValue=1.03, parameterName=Refraction, sliderMax=2}
 ]
}
