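# Secondary-motion AnimBP for the male medium Spy Tech Hacker body, layering procedural jacket, flap, cable and shoulder-strap motion onto a pose copied from the attached parent mesh.
# Look At nodes aim jacket bones, then spring Anim Dynamics simulate them against thigh sphere limits.
# Thigh-rotation alphas from the event graph scale front jacket bones; flaps and cables simulate afterward.
# Weighted constraints and Look Ats rig the left strap; wrist and crouch values go unused.
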AnimBlueprint /BRCosmetics/Characters/Player/Male/Medium/Bodies/M_MED_Spy_Tech_Hacker/Meshes/M_MED_Spy_Tech_Hacker_AnimBP
target_skeleton=&/M_MED_Spy_Tech_Hacker_Skeleton
graphs [
 AnimGraph {nodes [
   n1 = Look At - Bone: dyn_jacket_fr_l_01(ComponentPose=Local To Component(LocalPose=Copy Pose From Mesh() {bUseAttachedParent=true}), Alpha=1) {
    alphaBoolBlend {blendOption=Linear}
    boneToModify {boneName=dyn_jacket_fr_l_01}
    lookAtLocation {x=0}
    lookAtTarget {boneReference {boneName=dyn_jacket_fr_l_tgt}}
    lookAt_Axis {axis=<1 0>}
    lookUp_Axis {axis {x=0}}
   }
   n2 = Look At - Bone: dyn_jacket_fr_r_01(ComponentPose=@n1, Alpha=1) {
    alphaBoolBlend {blendOption=Linear}
    boneToModify {boneName=dyn_jacket_fr_r_01}
    lookAtLocation {x=0}
    lookAtTarget {boneReference {boneName=dyn_jacket_fr_r_tgt}}
    lookAt_Axis {axis=<1 0>}
    lookUp_Axis {axis {x=0}}
   }
   n3 = Look At - Bone: dyn_jacket_fr_l_02(ComponentPose=@n2, Alpha=0) {
    alphaBoolBlend {blendOption=Linear}
    boneToModify {boneName=dyn_jacket_fr_l_02}
    lookAtLocation {x=0}
    lookAtTarget {boneReference {boneName=dyn_jacket_fr_l_tgt}}
    lookAt_Axis {axis=<1 0>}
    lookUp_Axis {axis {x=0}}
   }
   n4 = Look At - Bone: dyn_jacket_fr_r_02(ComponentPose=@n3, Alpha=0) {
    alphaBoolBlend {blendOption=Linear}
    boneToModify {boneName=dyn_jacket_fr_r_02}
    lookAtLocation {x=0}
    lookAtTarget {boneReference {boneName=dyn_jacket_fr_r_tgt}}
    lookAt_Axis {axis=<1 0>}
    lookUp_Axis {axis {x=0}}
   }
   n5 = Look At - Bone: dyn_jacket_l_01(ComponentPose=@n4, Alpha=1) {
    alphaBoolBlend {blendOption=Linear}
    boneToModify {boneName=dyn_jacket_l_01}
    lookAtLocation {x=0}
    lookAtTarget {boneReference {boneName=dyn_jacket_l_tgt}}
    lookAt_Axis {axis=<1 0>}
    lookUp_Axis {axis {x=0}}
   }
   n6 = Look At - Bone: dyn_jacket_r_01(ComponentPose=@n5, Alpha=1) {
    alphaBoolBlend {blendOption=Linear}
    boneToModify {boneName=dyn_jacket_r_01}
    lookAtLocation {x=0}
    lookAtTarget {boneReference {boneName=dyn_jacket_r_tgt}}
    lookAt_Axis {axis=<1 0>}
    lookUp_Axis {axis {x=0}}
   }
   n7 = Anim Dynamics - Bone: dyn_jacket_fr_l_01(ComponentPose=@n6, Alpha=1, ExternalForce="0,0,0") {
    angularDampingOverride=1
    angularSpringConstant=1500
    bAngularSpring=true
    bGravityOverrideInSimSpace=true
    bOverrideAngularDamping=true
    bUseSphericalLimits=true
    gravityScale=0.35
    simulationSpace=RootRelative
    alphaBoolBlend {blendOption=Linear}
    alphaScaleBiasClamp {bInterpResult=true, scale=0.4}
    boundBone {boneName=dyn_jacket_fr_l_01}
    physicsBodyDefinitions [{
      collisionType=CustomSphere
      sphereCollisionRadius=2
      boundBone {boneName=dyn_jacket_fr_l_01}
      boxExtents=<10 20 20>
      constraintSetup {angularLimitsMax=<0 20 0>, angularLimitsMin=<0 -20 -40>, angularTarget=<0 0 0>}
      localJointOffset=<10 0 0>
     }]
    sphericalLimits [{limitRadius=12, drivingBone {boneName=thigh_l}, sphereLocalOffset=<-10 0 0>}]
   }
   n8 = Anim Dynamics - Bone: dyn_jacket_fr_l_02(ComponentPose=@n7, Alpha=1, ExternalForce="0,0,0") {
    angularDampingOverride=1
    angularSpringConstant=1500
    bAngularSpring=true
    bGravityOverrideInSimSpace=true
    bOverrideAngularDamping=true
    bUseSphericalLimits=true
    gravityScale=0.35
    simulationSpace=RootRelative
    alphaBoolBlend {blendOption=Linear}
    alphaScaleBiasClamp {bInterpResult=true, scale=0.7}
    boundBone {boneName=dyn_jacket_fr_l_02}
    physicsBodyDefinitions [{
      collisionType=CustomSphere
      sphereCollisionRadius=2
      boundBone {boneName=dyn_jacket_fr_l_02}
      boxExtents=<10 20 20>
      constraintSetup {angularLimitsMax=<0 20 0>, angularLimitsMin=<0 -20 -40>, angularTarget=<0 0 0>}
      localJointOffset=<10 0 0>
     }]
    sphericalLimits [{limitRadius=12, drivingBone {boneName=thigh_l}, sphereLocalOffset=<-20 0 0>}]
   }
   n9 = Anim Dynamics - Bone: dyn_jacket_l_01(ComponentPose=@n8, Alpha=1, ExternalForce="0,0,0") {
    angularDampingOverride=1
    angularSpringConstant=1500
    bAngularSpring=true
    bGravityOverrideInSimSpace=true
    bOverrideAngularDamping=true
    bUseSphericalLimits=true
    gravityScale=0.35
    simulationSpace=RootRelative
    alphaBoolBlend {blendOption=Linear}
    alphaScaleBiasClamp {scale=0.4}
    boundBone {boneName=dyn_jacket_l_01}
    physicsBodyDefinitions [{
      collisionType=CustomSphere
      sphereCollisionRadius=2
      boundBone {boneName=dyn_jacket_l_01}
      boxExtents=<10 20 20>
      constraintSetup {angularLimitsMax=<0 20 0>, angularLimitsMin=<0 -20 -40>, angularTarget=<0 0 0>}
      localJointOffset=<10 0 0>
     }]
    sphericalLimits [{limitRadius=12, drivingBone {boneName=thigh_l}, sphereLocalOffset=<-10 0 0>}]
   }
   n10 = Anim Dynamics - Bone: dyn_jacket_l_02(ComponentPose=@n9, Alpha=1, ExternalForce="0,0,0") {
    angularDampingOverride=1
    angularSpringConstant=1500
    bAngularSpring=true
    bGravityOverrideInSimSpace=true
    bOverrideAngularDamping=true
    bUseSphericalLimits=true
    gravityScale=0.35
    simulationSpace=RootRelative
    alphaBoolBlend {blendOption=Linear}
    alphaScaleBiasClamp {scale=0.7}
    boundBone {boneName=dyn_jacket_l_02}
    physicsBodyDefinitions [{
      collisionType=CustomSphere
      sphereCollisionRadius=2
      boundBone {boneName=dyn_jacket_l_02}
      boxExtents=<10 20 20>
      constraintSetup {angularLimitsMax=<0 20 0>, angularLimitsMin=<0 -20 -40>, angularTarget=<0 0 0>}
      localJointOffset=<10 0 0>
     }]
    sphericalLimits [{limitRadius=12, drivingBone {boneName=thigh_l}, sphereLocalOffset=<-20 0 0>}]
   }
   n11 = Anim Dynamics - Bone: dyn_jacket_bk_l_01(ComponentPose=@n10, Alpha=1, ExternalForce="0,0,0") {
    angularDampingOverride=1
    angularSpringConstant=1500
    bAngularSpring=true
    bGravityOverrideInSimSpace=true
    bOverrideAngularDamping=true
    bUseSphericalLimits=true
    gravityScale=0.35
    simulationSpace=RootRelative
    alphaBoolBlend {blendOption=Linear}
    alphaScaleBiasClamp {scale=0.4}
    boundBone {boneName=dyn_jacket_bk_l_01}
    physicsBodyDefinitions [{
      collisionType=CustomSphere
      sphereCollisionRadius=2
      boundBone {boneName=dyn_jacket_bk_l_01}
      boxExtents=<10 20 20>
      constraintSetup {angularLimitsMax=<0 20 0>, angularLimitsMin=<0 -20 -40>, angularTarget=<0 0 0>}
      localJointOffset=<10 0 0>
     }]
    sphericalLimits [{limitRadius=12, drivingBone {boneName=thigh_l}, sphereLocalOffset=<-10 0 0>}]
   }
   n12 = Anim Dynamics - Bone: dyn_jacket_bk_l_02(ComponentPose=@n11, Alpha=1, ExternalForce="0,0,0") {
    angularDampingOverride=1
    angularSpringConstant=1500
    bAngularSpring=true
    bGravityOverrideInSimSpace=true
    bOverrideAngularDamping=true
    bUseSphericalLimits=true
    gravityScale=0.35
    simulationSpace=RootRelative
    alphaBoolBlend {blendOption=Linear}
    alphaScaleBiasClamp {scale=0.7}
    boundBone {boneName=dyn_jacket_bk_l_02}
    physicsBodyDefinitions [{
      collisionType=CustomSphere
      sphereCollisionRadius=2
      boundBone {boneName=dyn_jacket_bk_l_02}
      boxExtents=<10 20 20>
      constraintSetup {angularLimitsMax=<0 20 0>, angularLimitsMin=<0 -20 -40>, angularTarget=<0 0 0>}
      localJointOffset=<10 0 0>
     }]
    sphericalLimits [{limitRadius=12, drivingBone {boneName=thigh_l}, sphereLocalOffset=<-20 0 0>}]
   }
   n13 = Anim Dynamics - Bone: dyn_jacket_bk2_l_01(ComponentPose=@n12, Alpha=1, ExternalForce="0,0,0") {
    angularDampingOverride=1
    angularSpringConstant=1500
    bAngularSpring=true
    bGravityOverrideInSimSpace=true
    bOverrideAngularDamping=true
    bUseSphericalLimits=true
    gravityScale=0.35
    simulationSpace=RootRelative
    alphaBoolBlend {blendOption=Linear}
    alphaScaleBiasClamp {scale=0.4}
    boundBone {boneName=dyn_jacket_bk2_l_01}
    physicsBodyDefinitions [{
      collisionType=CustomSphere
      sphereCollisionRadius=2
      boundBone {boneName=dyn_jacket_bk2_l_01}
      boxExtents=<10 20 20>
      constraintSetup {angularLimitsMax=<0 20 0>, angularLimitsMin=<0 -20 -40>, angularTarget=<0 0 0>}
      localJointOffset=<10 0 0>
     }]
    sphericalLimits [{limitRadius=12, drivingBone {boneName=thigh_l}, sphereLocalOffset=<-10 0 0>}]
   }
   n14 = Anim Dynamics - Bone: dyn_jacket_bk2_l_02(ComponentPose=@n13, Alpha=1, ExternalForce="0,0,0") {
    angularDampingOverride=1
    angularSpringConstant=1500
    bAngularSpring=true
    bGravityOverrideInSimSpace=true
    bOverrideAngularDamping=true
    bUseSphericalLimits=true
    gravityScale=0.35
    simulationSpace=RootRelative
    alphaBoolBlend {blendOption=Linear}
    alphaScaleBiasClamp {scale=0.7}
    boundBone {boneName=dyn_jacket_bk2_l_02}
    physicsBodyDefinitions [{
      collisionType=CustomSphere
      sphereCollisionRadius=2
      boundBone {boneName=dyn_jacket_bk2_l_02}
      boxExtents=<10 20 20>
      constraintSetup {angularLimitsMax=<0 20 0>, angularLimitsMin=<0 -20 -40>, angularTarget=<0 0 0>}
      localJointOffset=<10 0 0>
     }]
    sphericalLimits [{limitRadius=12, drivingBone {boneName=thigh_l}, sphereLocalOffset=<-20 0 0>}]
   }
   n15 = Anim Dynamics - Bone: dyn_jacket_bk2_r_01(ComponentPose=@n14, Alpha=1, ExternalForce="0,0,0") {
    angularDampingOverride=1
    angularSpringConstant=1500
    bAngularSpring=true
    bGravityOverrideInSimSpace=true
    bOverrideAngularDamping=true
    bUseSphericalLimits=true
    gravityScale=0.35
    simulationSpace=RootRelative
    alphaBoolBlend {blendOption=Linear}
    alphaScaleBiasClamp {scale=0.4}
    boundBone {boneName=dyn_jacket_bk2_r_01}
    physicsBodyDefinitions [{
      collisionType=CustomSphere
      sphereCollisionRadius=2
      boundBone {boneName=dyn_jacket_bk2_r_01}
      boxExtents=<10 20 20>
      constraintSetup {angularLimitsMax=<0 20 0>, angularLimitsMin=<0 -20 -40>, angularTarget=<0 0 0>}
      localJointOffset=<10 0 0>
     }]
    sphericalLimits [{limitRadius=12, drivingBone {boneName=thigh_r}, sphereLocalOffset=<10 0 0>}]
   }
   n16 = Anim Dynamics - Bone: dyn_jacket_bk2_r_02(ComponentPose=@n15, Alpha=1, ExternalForce="0,0,0") {
    angularDampingOverride=1
    angularSpringConstant=1500
    bAngularSpring=true
    bGravityOverrideInSimSpace=true
    bOverrideAngularDamping=true
    bUseSphericalLimits=true
    gravityScale=0.35
    simulationSpace=RootRelative
    alphaBoolBlend {blendOption=Linear}
    alphaScaleBiasClamp {scale=0.7}
    boundBone {boneName=dyn_jacket_bk2_r_02}
    physicsBodyDefinitions [{
      collisionType=CustomSphere
      sphereCollisionRadius=2
      boundBone {boneName=dyn_jacket_bk2_r_02}
      boxExtents=<10 20 20>
      constraintSetup {angularLimitsMax=<0 20 0>, angularLimitsMin=<0 -20 -40>, angularTarget=<0 0 0>}
      localJointOffset=<10 0 0>
     }]
    sphericalLimits [{limitRadius=12, drivingBone {boneName=thigh_r}, sphereLocalOffset=<20 0 0>}]
   }
   n17 = Anim Dynamics - Bone: dyn_jacket_bk_r_01(ComponentPose=@n16, Alpha=1, ExternalForce="0,0,0") {
    angularDampingOverride=1
    angularSpringConstant=1500
    bAngularSpring=true
    bGravityOverrideInSimSpace=true
    bOverrideAngularDamping=true
    bUseSphericalLimits=true
    gravityScale=0.35
    simulationSpace=RootRelative
    alphaBoolBlend {blendOption=Linear}
    alphaScaleBiasClamp {scale=0.4}
    boundBone {boneName=dyn_jacket_bk_r_01}
    physicsBodyDefinitions [{
      collisionType=CustomSphere
      sphereCollisionRadius=2
      boundBone {boneName=dyn_jacket_bk_r_01}
      boxExtents=<10 20 20>
      constraintSetup {angularLimitsMax=<0 20 0>, angularLimitsMin=<0 -20 -40>, angularTarget=<0 0 0>}
      localJointOffset=<10 0 0>
     }]
    sphericalLimits [{limitRadius=12, drivingBone {boneName=thigh_r}, sphereLocalOffset=<10 0 0>}]
   }
   n18 = Anim Dynamics - Bone: dyn_jacket_bk_r_02(ComponentPose=@n17, Alpha=1, ExternalForce="0,0,0") {
    angularDampingOverride=1
    angularSpringConstant=1500
    bAngularSpring=true
    bGravityOverrideInSimSpace=true
    bOverrideAngularDamping=true
    bUseSphericalLimits=true
    gravityScale=0.35
    simulationSpace=RootRelative
    alphaBoolBlend {blendOption=Linear}
    alphaScaleBiasClamp {scale=0.7}
    boundBone {boneName=dyn_jacket_bk_r_02}
    physicsBodyDefinitions [{
      collisionType=CustomSphere
      sphereCollisionRadius=2
      boundBone {boneName=dyn_jacket_bk_r_02}
      boxExtents=<10 20 20>
      constraintSetup {angularLimitsMax=<0 20 0>, angularLimitsMin=<0 -20 -40>, angularTarget=<0 0 0>}
      localJointOffset=<10 0 0>
     }]
    sphericalLimits [{limitRadius=12, drivingBone {boneName=thigh_r}, sphereLocalOffset=<20 0 0>}]
   }
   n19 = Anim Dynamics - Bone: dyn_jacket_r_01(ComponentPose=@n18, Alpha=1, ExternalForce="0,0,0") {
    angularDampingOverride=1
    angularSpringConstant=1500
    bAngularSpring=true
    bGravityOverrideInSimSpace=true
    bOverrideAngularDamping=true
    bUseSphericalLimits=true
    gravityScale=0.35
    simulationSpace=RootRelative
    alphaBoolBlend {blendOption=Linear}
    alphaScaleBiasClamp {scale=0.4}
    boundBone {boneName=dyn_jacket_r_01}
    physicsBodyDefinitions [{
      collisionType=CustomSphere
      sphereCollisionRadius=2
      boundBone {boneName=dyn_jacket_r_01}
      boxExtents=<10 20 20>
      constraintSetup {angularLimitsMax=<0 20 0>, angularLimitsMin=<0 -20 -40>, angularTarget=<0 0 0>}
      localJointOffset=<10 0 0>
     }]
    sphericalLimits [{limitRadius=12, drivingBone {boneName=thigh_r}, sphereLocalOffset=<10 0 0>}]
   }
   n20 = Anim Dynamics - Bone: dyn_jacket_r_02(ComponentPose=@n19, Alpha=1, ExternalForce="0,0,0") {
    angularDampingOverride=1
    angularSpringConstant=1500
    bAngularSpring=true
    bGravityOverrideInSimSpace=true
    bOverrideAngularDamping=true
    bUseSphericalLimits=true
    gravityScale=0.35
    simulationSpace=RootRelative
    alphaBoolBlend {blendOption=Linear}
    alphaScaleBiasClamp {scale=0.7}
    boundBone {boneName=dyn_jacket_r_02}
    physicsBodyDefinitions [{
      collisionType=CustomSphere
      sphereCollisionRadius=2
      boundBone {boneName=dyn_jacket_r_02}
      boxExtents=<10 20 20>
      constraintSetup {angularLimitsMax=<0 20 0>, angularLimitsMin=<0 -20 -40>, angularTarget=<0 0 0>}
      localJointOffset=<10 0 0>
     }]
    sphericalLimits [{limitRadius=12, drivingBone {boneName=thigh_r}, sphereLocalOffset=<20 0 0>}]
   }
   n21 = Anim Dynamics - Bone: dyn_jacket_fr_r_01(ComponentPose=@n20, Alpha=1, ExternalForce="0,0,0") {
    angularDampingOverride=1
    angularSpringConstant=1500
    bAngularSpring=true
    bGravityOverrideInSimSpace=true
    bOverrideAngularDamping=true
    bUseSphericalLimits=true
    gravityScale=0.35
    simulationSpace=RootRelative
    alphaBoolBlend {blendOption=Linear}
    alphaScaleBiasClamp {scale=0.4}
    boundBone {boneName=dyn_jacket_fr_r_01}
    physicsBodyDefinitions [{
      collisionType=CustomSphere
      sphereCollisionRadius=2
      boundBone {boneName=dyn_jacket_fr_r_01}
      boxExtents=<10 20 20>
      constraintSetup {angularLimitsMax=<0 20 0>, angularLimitsMin=<0 -20 -40>, angularTarget=<0 0 0>}
      localJointOffset=<10 0 0>
     }]
    sphericalLimits [{limitRadius=12, drivingBone {boneName=thigh_r}, sphereLocalOffset=<10 0 0>}]
   }
   n22 = Anim Dynamics - Bone: dyn_jacket_fr_r_02(ComponentPose=@n21, Alpha=1, ExternalForce="0,0,0") {
    angularDampingOverride=1
    angularSpringConstant=1500
    bAngularSpring=true
    bGravityOverrideInSimSpace=true
    bOverrideAngularDamping=true
    bUseSphericalLimits=true
    gravityScale=0.35
    simulationSpace=RootRelative
    alphaBoolBlend {blendOption=Linear}
    alphaScaleBiasClamp {scale=0.7}
    boundBone {boneName=dyn_jacket_fr_r_02}
    physicsBodyDefinitions [{
      collisionType=CustomSphere
      sphereCollisionRadius=2
      boundBone {boneName=dyn_jacket_fr_r_02}
      boxExtents=<10 20 20>
      constraintSetup {angularLimitsMax=<0 20 0>, angularLimitsMin=<0 -20 -40>, angularTarget=<0 0 0>}
      localJointOffset=<10 0 0>
     }]
    sphericalLimits [{limitRadius=12, drivingBone {boneName=thigh_r}, sphereLocalOffset=<20 0 0>}]
   }
   n23 = Get thigh_rot_l()
   n24 = "Transform (Modify) Bone - Bone: dyn_jacket_fr_l_01"(ComponentPose=@n22, Alpha=@n23, Translation="0,0,0", Rotation="0,0,0", Scale="0.5,1,1") {
    rotationMode=BMM_Additive
    rotationSpace=BCS_BoneSpace
    scaleMode=BMM_Additive
    scaleSpace=BCS_BoneSpace
    translationMode=BMM_Additive
    translationSpace=BCS_BoneSpace
    alphaBoolBlend {blendOption=Linear}
    alphaScaleBiasClamp {bInterpResult=true}
    boneToModify {boneName=dyn_jacket_fr_l_01}
    scale {x=0.5}
   }
   n25 = "Transform (Modify) Bone - Bone: dyn_jacket_fr_l_02"(ComponentPose=@n24, Alpha=@n23, Translation="0,0,0", Rotation="0,0,0", Scale="0.5,1,1") {
    rotationMode=BMM_Additive
    rotationSpace=BCS_BoneSpace
    scaleMode=BMM_Additive
    scaleSpace=BCS_BoneSpace
    translationMode=BMM_Additive
    translationSpace=BCS_ParentBoneSpace
    alphaBoolBlend {blendOption=Linear}
    alphaScaleBiasClamp {bInterpResult=true}
    boneToModify {boneName=dyn_jacket_fr_l_02}
    scale {x=0.5}
   }
   n26 = Get thigh_rot_r()
   n27 = "Transform (Modify) Bone - Bone: dyn_jacket_fr_r_01"(ComponentPose=@n25, Alpha=@n26, Translation="0,0,0", Rotation="0,0,0", Scale="0.5,1,1") {
    rotationMode=BMM_Additive
    rotationSpace=BCS_BoneSpace
    scaleMode=BMM_Additive
    scaleSpace=BCS_BoneSpace
    translationMode=BMM_Additive
    translationSpace=BCS_BoneSpace
    alphaBoolBlend {blendOption=Linear}
    alphaScaleBiasClamp {bInterpResult=true}
    boneToModify {boneName=dyn_jacket_fr_r_01}
    scale {x=0.5}
   }
   n28 = "Transform (Modify) Bone - Bone: dyn_jacket_fr_r_02"(ComponentPose=@n27, Alpha=@n26, Translation="0,0,0", Rotation="0,0,0", Scale="0.5,1,1") {
    rotationMode=BMM_Additive
    rotationSpace=BCS_BoneSpace
    scaleMode=BMM_Additive
    scaleSpace=BCS_BoneSpace
    translationMode=BMM_Additive
    translationSpace=BCS_ParentBoneSpace
    alphaBoolBlend {blendOption=Linear}
    alphaScaleBiasClamp {bInterpResult=true}
    boneToModify {boneName=dyn_jacket_fr_r_02}
    scale {x=0.5}
   }
   n29 = Anim Dynamics - Bone: flap(ComponentPose=@n28, Alpha=1, ExternalForce="0,0,0") {
    angularDampingOverride=1
    angularSpringConstant=1500
    bAngularSpring=true
    bGravityOverrideInSimSpace=true
    bOverrideAngularDamping=true
    gravityScale=0.35
    simulationSpace=RootRelative
    alphaBoolBlend {blendOption=Linear}
    alphaScaleBiasClamp {scale=0.15}
    boundBone {boneName=flap}
    physicsBodyDefinitions [{
      sphereCollisionRadius=0
      boundBone {boneName=flap}
      boxExtents=<10 20 20>
      constraintSetup {angularLimitsMax=<0 10 0>, angularLimitsMin=<0 -10 -40>, angularTarget=<0 0 0>}
      localJointOffset=<10 0 0>
     }]
   }
   n30 = Anim Dynamics - Bone: flap_2(ComponentPose=@n29, Alpha=1, ExternalForce="0,0,0") {
    angularDampingOverride=1
    angularSpringConstant=1500
    bAngularSpring=true
    bGravityOverrideInSimSpace=true
    bOverrideAngularDamping=true
    gravityScale=0.35
    simulationSpace=RootRelative
    alphaBoolBlend {blendOption=Linear}
    alphaScaleBiasClamp {scale=0.15}
    boundBone {boneName=flap_2}
    physicsBodyDefinitions [{
      sphereCollisionRadius=0
      boundBone {boneName=flap_2}
      boxExtents=<10 20 20>
      constraintSetup {angularLimitsMax=<0 10 0>, angularLimitsMin=<0 -10 -40>, angularTarget=<0 0 0>}
      localJointOffset=<10 0 0>
     }]
   }
   n31 = Anim Dynamics - Bone: cable_1(ComponentPose=@n30, Alpha=1, ExternalForce="0,0,0") {
    angularDampingOverride=1
    angularSpringConstant=1500
    bAngularSpring=true
    bGravityOverrideInSimSpace=true
    bOverrideAngularDamping=true
    gravityScale=0.35
    simulationSpace=RootRelative
    alphaBoolBlend {blendOption=Linear}
    boundBone {boneName=cable_1}
    physicsBodyDefinitions [{
      sphereCollisionRadius=0
      boundBone {boneName=cable_1}
      boxExtents=<10 20 20>
      constraintSetup {angularLimitsMax=<0 20 0>, angularLimitsMin=<0 -20 -40>, angularTarget=<0 0 0>}
      localJointOffset=<10 0 0>
     }]
   }
   n32 = Anim Dynamics - Bone: cable_2(ComponentPose=@n31, Alpha=1, ExternalForce="0,0,0") {
    angularDampingOverride=1
    angularSpringConstant=1500
    bAngularSpring=true
    bGravityOverrideInSimSpace=true
    bOverrideAngularDamping=true
    gravityScale=0.35
    simulationSpace=RootRelative
    alphaBoolBlend {blendOption=Linear}
    boundBone {boneName=cable_2}
    physicsBodyDefinitions [{
      sphereCollisionRadius=0
      boundBone {boneName=cable_2}
      boxExtents=<10 20 20>
      constraintSetup {angularLimitsMax=<0 20 0>, angularLimitsMin=<0 -20 -40>, angularTarget=<0 0 0>}
      localJointOffset=<10 0 0>
     }]
   }
   n33 = "Constraint - shoulder_strap_aim_4_l_zero (PPPPPP)"!(
    ComponentPose=@n32
    Alpha=1
    ConstraintWeights_0=0.263
    ConstraintWeights_1:float=0.086
    ConstraintWeights_2:float=0.022
    ConstraintWeights_3:float=0.528
    ConstraintWeights_4:float=0.072
    ConstraintWeights_5:float=0.028
   ) -> (Pose) {
    alphaBoolBlend {blendOption=Linear}
    boneToModify {boneName=shoulder_strap_aim_4_l_zero}
    constraintSetup [
     {targetBone {boneName=clavicle_l}}
     {targetBone {boneName=spine_03}}
     {targetBone {boneName=spine_04}}
     {targetBone {boneName=spine_05}}
     {targetBone {boneName=neck_02}}
     {targetBone {boneName=neck_01}}
    ]
    constraintWeights [1,1,1,1,1,1]
   }
   n34 = "Constraint - shoulder_strap_aim_2_l_zero (PPPPPP)"!(
    ComponentPose=@n33
    Alpha=1
    ConstraintWeights_0=0.093
    ConstraintWeights_1:float=0.033
    ConstraintWeights_2:float=0.08
    ConstraintWeights_3:float=0.631
    ConstraintWeights_4:float=0.114
    ConstraintWeights_5:float=0.048
   ) -> (Pose) {
    alphaBoolBlend {blendOption=Linear}
    boneToModify {boneName=shoulder_strap_aim_2_l_zero}
    constraintSetup [
     {targetBone {boneName=clavicle_l}}
     {targetBone {boneName=spine_03}}
     {targetBone {boneName=spine_04}}
     {targetBone {boneName=spine_05}}
     {targetBone {boneName=neck_02}}
     {targetBone {boneName=neck_01}}
    ]
    constraintWeights [1,1,1,1,1,1]
   }
   n35 = "Constraint - shoulder_strap_aim_1_l_zero (PPPP)"!(
    ComponentPose=@n34
    Alpha=1
    ConstraintWeights_0=0.111
    ConstraintWeights_1:float=0.445
    ConstraintWeights_2:float=0.421
    ConstraintWeights_3:float=0.023
   ) -> (Pose) {
    alphaBoolBlend {blendOption=Linear}
    boneToModify {boneName=shoulder_strap_aim_1_l_zero}
    constraintSetup [{targetBone {boneName=spine_03}}, {targetBone {boneName=spine_04}}, {targetBone {boneName=spine_05}}, {targetBone {boneName=clavicle_l}}]
    constraintWeights [1,1,1,1]
   }
   n36 = "Constraint - shoulder_strap_aim_4_l_tgt (PPPP)"!(
    ComponentPose=@n35
    Alpha=1
    ConstraintWeights_0=0.053
    ConstraintWeights_1:float=0.438
    ConstraintWeights_2:float=0.473
    ConstraintWeights_3:float=0.036
   ) -> (Pose) {
    alphaBoolBlend {blendOption=Linear}
    boneToModify {boneName=shoulder_strap_aim_4_l_tgt}
    constraintSetup [
     {targetBone {boneName=spine_03}}
     {targetBone {boneName=clavicle_l}}
     {targetBone {boneName=upperarm_twist_01_l}}
     {targetBone {boneName=spine_05}}
    ]
    constraintWeights [1,1,1,1]
   }
   n37 = "Constraint - shoulder_strap_aim_2_l_tgt (PPPP)"!(
    ComponentPose=@n36
    Alpha=1
    ConstraintWeights_0=0.036
    ConstraintWeights_1:float=0.369
    ConstraintWeights_2:float=0.554
    ConstraintWeights_3:float=0.041
   ) -> (Pose) {
    alphaBoolBlend {blendOption=Linear}
    boneToModify {boneName=shoulder_strap_aim_2_l_tgt}
    constraintSetup [
     {targetBone {boneName=spine_03}}
     {targetBone {boneName=clavicle_l}}
     {targetBone {boneName=upperarm_twist_01_l}}
     {targetBone {boneName=spine_05}}
    ]
    constraintWeights [1,1,1,1]
   }
   n38 = "Constraint - shoulder_strap_aim_1_l_tgt (PPPPP)"!(
    ComponentPose=@n37
    Alpha=1
    ConstraintWeights_0=0.06
    ConstraintWeights_1:float=0.131
    ConstraintWeights_2:float=0.229
    ConstraintWeights_3:float=0.543
    ConstraintWeights_4:float=0.036
   ) -> (Pose) {
    alphaBoolBlend {blendOption=Linear}
    boneToModify {boneName=shoulder_strap_aim_1_l_tgt}
    constraintSetup [
     {targetBone {boneName=spine_03}}
     {targetBone {boneName=spine_04}}
     {targetBone {boneName=clavicle_l}}
     {targetBone {boneName=upperarm_twist_01_l}}
     {targetBone {boneName=spine_05}}
    ]
    constraintWeights [1,1,1,1,1]
   }
   n39 = "Constraint - shoulder_strap_aim_3_l_tgt (PPPP)"!(
    ComponentPose=@n38
    Alpha=1
    ConstraintWeights_0=0.037
    ConstraintWeights_1:float=0.423
    ConstraintWeights_2:float=0.496
    ConstraintWeights_3:float=0.044
   ) -> (Pose) {
    alphaBoolBlend {blendOption=Linear}
    boneToModify {boneName=shoulder_strap_aim_3_l_tgt}
    constraintSetup [
     {targetBone {boneName=spine_03}}
     {targetBone {boneName=upperarm_twist_01_l}}
     {targetBone {boneName=clavicle_l}}
     {targetBone {boneName=spine_05}}
    ]
    constraintWeights [1,1,1,1]
   }
   n40 = "Constraint - shoulder_strap_l_zero (PPPPPPPP)"!(
    ComponentPose=@n39
    Alpha=1
    ConstraintWeights_0=0.021
    ConstraintWeights_1:float=0.186
    ConstraintWeights_2:float=0.014
    ConstraintWeights_3:float=0.016
    ConstraintWeights_4:float=0.513
    ConstraintWeights_5:float=0.011
    ConstraintWeights_6:float=0.17
    ConstraintWeights_7:float=0.069
   ) -> (Pose) {
    alphaBoolBlend {blendOption=Linear}
    boneToModify {boneName=shoulder_strap_l_zero}
    constraintSetup [
     {targetBone {boneName=head}}
     {targetBone {boneName=clavicle_l}}
     {targetBone {boneName=spine_03}}
     {targetBone {boneName=spine_04}}
     {targetBone {boneName=spine_05}}
     {targetBone {boneName=upperarm_twist_01_l}}
     {targetBone {boneName=neck_02}}
     {targetBone {boneName=neck_01}}
    ]
    constraintWeights [1,1,1,1,1,1,1,1]
   }
   n41 = Look At - Bone: shoulder_strap_aim_1_l(ComponentPose=@n40, Alpha=1) {
    alphaBoolBlend {blendOption=Linear}
    boneToModify {boneName=shoulder_strap_aim_1_l}
    lookAtLocation {x=0}
    lookAtTarget {boneReference {boneName=shoulder_strap_aim_1_l_tgt}}
    lookAt_Axis {axis=<1 0>}
    lookUp_Axis {axis {x=0}}
   }
   n42 = Look At - Bone: shoulder_strap_aim_2_l(ComponentPose=@n41, Alpha=1) {
    alphaBoolBlend {blendOption=Linear}
    boneToModify {boneName=shoulder_strap_aim_2_l}
    lookAtLocation {x=0}
    lookAtTarget {boneReference {boneName=shoulder_strap_aim_2_l_tgt}}
    lookAt_Axis {axis=<1 0>}
    lookUp_Axis {axis {x=0}}
   }
   n43 = Look At - Bone: shoulder_strap_aim_3_l(ComponentPose=@n42, Alpha=1) {
    alphaBoolBlend {blendOption=Linear}
    boneToModify {boneName=shoulder_strap_aim_3_l}
    lookAtLocation {x=0}
    lookAtTarget {boneReference {boneName=shoulder_strap_aim_3_l_tgt}}
    lookAt_Axis {axis=<1 0>}
    lookUp_Axis {axis {x=0}}
   }
   n44 = Look At - Bone: shoulder_strap_aim_4_l(ComponentPose=@n43, Alpha=1) {
    alphaBoolBlend {blendOption=Linear}
    boneToModify {boneName=shoulder_strap_aim_4_l}
    lookAtLocation {x=0}
    lookAtTarget {boneReference {boneName=shoulder_strap_aim_4_l_tgt}}
    lookAt_Axis {axis=<1 0>}
    lookUp_Axis {axis {x=0}}
   }
   n45 = "Constraint - com_bottom (PPPPPPPPPPPPP)"!(
    ComponentPose=@n44
    Alpha=1
    ConstraintWeights_0=0.004
    ConstraintWeights_1:float=0.003
    ConstraintWeights_2:float=0.042
    ConstraintWeights_3:float=0.1
    ConstraintWeights_4:float=0.099
    ConstraintWeights_5:float=0.07
    ConstraintWeights_6:float=0.268
    ConstraintWeights_7:float=0.143
    ConstraintWeights_8:float=0.074
    ConstraintWeights_9:float=0.007
    ConstraintWeights_10:float=0.094
    ConstraintWeights_11:float=0.095
    ConstraintWeights_12:float=0.002
   ) -> (Pose) {
    alphaBoolBlend {blendOption=Linear}
    boneToModify {boneName=com_bottom}
    constraintSetup [
     {targetBone {boneName=shoulder_strap_aim_3_l}}
     {targetBone {boneName=shoulder_strap_aim_3_l_tgt}}
     {targetBone {boneName=upperarm_twist_01_l}}
     {targetBone {boneName=clavicle_l}}
     {targetBone {boneName=shoulder_strap_aim_2_l_tgt}}
     {targetBone {boneName=spine_03}}
     {targetBone {boneName=spine_04}}
     {targetBone {boneName=spine_05}}
     {targetBone {boneName=shoulder_strap_aim_1_l}}
     {targetBone {boneName=neck_02}}
     {targetBone {boneName=shoulder_strap_aim_2_l}}
     {targetBone {boneName=shoulder_strap_aim_1_l_tgt}}
     {targetBone {boneName=neck_01}}
    ]
    constraintWeights [1,1,1,1,1,1,1,1,1,1,1,1,1]
   }
   n46 = "Constraint - com_top_zero (PPPPPPPPPPPPPP)"!(
    ComponentPose=@n45
    Alpha=1
    ConstraintWeights_0=0.155
    ConstraintWeights_1:float=0.11
    ConstraintWeights_2:float=0.002
    ConstraintWeights_3:float=0.02
    ConstraintWeights_4:float=0.002
    ConstraintWeights_5:float=0.234
    ConstraintWeights_6:float=0.003
    ConstraintWeights_7:float=0.011
    ConstraintWeights_8:float=0.101
    ConstraintWeights_9:float=0.01
    ConstraintWeights_10:float=0.027
    ConstraintWeights_11:float=0.301
    ConstraintWeights_12:float=0.012
    ConstraintWeights_13:float=0.012
   ) -> (Pose) {
    alphaBoolBlend {blendOption=Linear}
    boneToModify {boneName=com_top_zero}
    constraintSetup [
     {targetBone {boneName=shoulder_strap_aim_3_l}}
     {targetBone {boneName=shoulder_strap_aim_3_l_tgt}}
     {targetBone {boneName=upperarm_twist_01_l}}
     {targetBone {boneName=clavicle_l}}
     {targetBone {boneName=shoulder_strap_aim_4_l}}
     {targetBone {boneName=shoulder_strap_aim_2_l_tgt}}
     {targetBone {boneName=spine_03}}
     {targetBone {boneName=spine_04}}
     {targetBone {boneName=spine_05}}
     {targetBone {boneName=shoulder_strap_aim_1_l}}
     {targetBone {boneName=neck_02}}
     {targetBone {boneName=shoulder_strap_aim_2_l}}
     {targetBone {boneName=shoulder_strap_aim_1_l_tgt}}
     {targetBone {boneName=neck_01}}
    ]
    constraintWeights [1,1,1,1,1,1,1,1,1,1,1,1,1,1]
   }
   n47 = Look At - Bone: com_top(ComponentPose=@n46, Alpha=1) {
    alphaBoolBlend {blendOption=Linear}
    boneToModify {boneName=com_top}
    lookAtLocation {x=0}
    lookAtTarget {boneReference {boneName=com_bottom}}
    lookAt_Axis {axis=<1 0>}
    lookUp_Axis {axis {x=0}}
   }
   Output Pose(Result=Component To Local(ComponentPose=@n47)) {layerGroup=None}
   Reroute Node(InputPin:Wildcard) -> (OutputPin:Wildcard)
  ]}
 EventGraph {nodes [
   Event BlueprintUpdateAnimation()
   > Set wrist_bend_l(wrist_bend_l:double=GetAlphaFromReferenceTransformDelta(
     SkeletalMeshComponent=GetOwnerSkeletalMeshComponent()
     BoneName=hand_l
     BaseName=lowerarm_twist_02_l
     DeltaComponentType=QuaternionTwist
     InputRange=(X=0,Y=1)
     TwistAxis="0,0,1"
    ).RawOutput) -> (Output_Get:double)
   n1 = Get bIsCrouching()
   > Set crouch_alpha(crouch_alpha:double="To Float (Boolean)"(InBool=@n1)) -> (Output_Get:double)
   > Set no_crouch_alpha(no_crouch_alpha:double="To Float (Boolean)"(InBool=NOT Boolean(A=@n1))) -> (Output_Get:double)
   > Set thigh_rot_r(thigh_rot_r=GetAlphaFromReferenceTransformDelta(SkeletalMeshComponent=GetOwnerSkeletalMeshComponent(), BoneName=thigh_r, BaseName=pelvis, InputRange=(X=-20,Y=-100), TwistAxis="1,0,0"))
   > Set thigh_rot_l(thigh_rot_l=GetAlphaFromReferenceTransformDelta(SkeletalMeshComponent=GetOwnerSkeletalMeshComponent(), BoneName=thigh_l, BaseName=pelvis, InputRange=(X=-20,Y=-100), TwistAxis="1,0,0"))
  ]}
]
variables [
 crouch_alpha {type=number}
 no_crouch_alpha {type=number}
 thigh_rot_l {type=number}
 thigh_rot_r {type=number}
 wrist_bend_l {type=number}
]
